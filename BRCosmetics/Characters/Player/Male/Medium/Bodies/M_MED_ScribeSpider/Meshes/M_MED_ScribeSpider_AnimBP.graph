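# Animation Blueprint for the M_MED_ScribeSpider player body cosmetic, layering a Control Rig over the pose copied from its attached parent mesh.
# The event graph tracks hang gliding from parachute events and back items, including capes and long ones, on customization load.
# Each update flags LODs below 2 and any montage playing on the owner's anim instance as a custom emote.
# The emote flag visibly feeds the rig; other rig inputs show false.
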AnimBlueprint /BRCosmetics/Characters/Player/Male/Medium/Bodies/M_MED_ScribeSpider/Meshes/M_MED_ScribeSpider_AnimBP
target_skeleton=&/M_MED_ScribeSpider_Skeleton
graphs [
 AnimGraph {nodes [
   n1 = Control Rig!(
    Source=Copy Pose From Mesh() {bUseAttachedParent=true}
    Alpha
    isWearingBackbling:bool=false
    LowLods:bool=false
    isHangGliding:bool=false
    Tail Corrective:bool=false
    isWearingShoes:bool=false
    Emote:bool=Get Is Custom Emote Playing() -> (Is Custom Emote Playing:bool)
   ) -> (Pose) {controlRigAssetReference {blueprintRigClass=&/M_MED_ScribeSpider_Ctrl_Rig.M_MED_ScribeSpider_Ctrl_Rig_C}}
   Output Pose(Result=@n1) {layerGroup=None}
  ]}
 EventGraph {nodes [
   Event OnParachuteOpened()
   > Switch on EGliderType(Selection=Get PlayerGliderType() -> (PlayerGliderType=HangGlider))
   > Set Is Hang Gliding(Is Hang Gliding:bool=true) -> (Output_Get:bool)
   Event OnParachuteClosed()
   > Set Is Hang Gliding(Is Hang Gliding=false)
   Event OnCharacterCustomizationLoaded()
   n1 = CheckForBackbling()
   > Set Is Wearing Back Item(Is Wearing Back Item:bool=OR Boolean!(A=@n1, B=@n1.bIsBackblingACape, C:bool=@n1.bIsBackblingLong) -> (ReturnValue)) -> (Output_Get:bool)
   Event BlueprintUpdateAnimation()
   > n2 = Sequence()
   n3 = GetAnimInstance(self=GetOwnerSkeletalMeshComponent())
   > Is Valid(InputObject=@n3)
   > Set Is Custom Emote Playing(Is Custom Emote Playing:bool=Montage_IsPlaying(self=@n3)) -> (Output_Get:bool)
   Set LowLODs(execute=@n2.then_1, LowLODs:bool="integer < integer"(A=Get CurrentLOD(), B=2)) -> (Output_Get:bool)
  ]}
]
variables [
 Is Custom Emote Playing {type=boolean}
 Is Hang Gliding {type=boolean}
 Is Wearing Back Item {type=boolean}
 isWearingShoes {type=boolean}
 LowLODs {type=boolean, default=true}
 Tail Corrective {type=boolean}
]
